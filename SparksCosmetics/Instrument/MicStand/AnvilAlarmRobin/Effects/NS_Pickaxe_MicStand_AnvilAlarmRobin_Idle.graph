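# Niagara idle effect for the AnvilAlarmRobin mic stand pickaxe cosmetic, built from one Minimal emitter with a mesh renderer.
# An instantaneous burst spawns the oriented, orange-yellow mesh, which persists past its lifetime and slowly counter-rotates about the Z axis.
# System-level variable assignments accompany it, with color-curve dynamic inputs among the dependencies and a front-end preview user flag exposed.
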NiagaraSystem /SparksCosmetics/Instrument/MicStand/AnvilAlarmRobin/Effects/NS_Pickaxe_MicStand_AnvilAlarmRobin_Idle
dependencies {
 usedDataInterfaces [&NiagaraDataInterfaceColorCurve]
 usedDynamicInputs [&ColorFromCurve, &MakeFloatFromVector]
 usedModules [
  &SystemState
  &self:SystemSpawnScript.SystemScriptSource.SystemScriptGraph.NiagaraNodeAssignment_1.SetVariables_#1
  &self:SystemSpawnScript.SystemScriptSource.SystemScriptGraph.NiagaraNodeAssignment_2.SetVariables_#2
 ]
}
emitters {Minimal {
  emitterUpdateScript [
   &EmitterState()
   &SpawnBurst_Instantaneous(Use Spawn Probability=false, UseLoopCountLimit=true)
  ]
  particleSpawnScript [
   &InitializeParticle(Color=<rgba 1 0.6887 0 1>, Lifetime=5, Mesh Scale Mode=::NewEnumerator3, UsePositionOffset=false, Write Material Random=true)
   &InitialMeshOrientation(Mesh Axis To Orient=<0 0.6 1>, Mesh Orientation Mode=::NewEnumerator2, Orientation Vector=<0 0 1>, Use Rotation Vector=false)
  ]
  particleUpdateScript [
   &ParticleState(Kill Particles When Lifetime Has Elapsed=false)
   &UpdateMeshOrientation(Rotation Rate=-0.05, Rotation Vector=<0 0 1>)
  ]
  rendererClasses [&NiagaraMeshRendererProperties]
  renderers [{rendererClass=&NiagaraMeshRendererProperties, rendererIndex=0}]
 }}
summary {
 systemName=NS_Pickaxe_MicStand_AnvilAlarmRobin_Idle
 emitters [{emitterName=Minimal, rendererClasses [&NiagaraMeshRendererProperties]}]
 userVariables [User.bIsFrontEndPreview {type=NiagaraBool}]
}
