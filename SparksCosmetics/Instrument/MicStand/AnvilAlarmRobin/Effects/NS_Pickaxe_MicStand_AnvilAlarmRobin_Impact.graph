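# Niagara impact effect for the AnvilAlarmRobin mic stand pickaxe from SparksCosmetics.
# Burst emitters include a yellow core flash with a light, surface-dependent debris driven by User.PhysicalSurface, heart sprites, smoke, sparks and spikes.
# The flash packs several looks into one emitter through spawn-group switches.
# System Global Scale, time-of-day color and brightness curves, and front-end versus gameplay modifiers adjust size, tint and intensity.
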NiagaraSystem /SparksCosmetics/Instrument/MicStand/AnvilAlarmRobin/Effects/NS_Pickaxe_MicStand_AnvilAlarmRobin_Impact
dependencies {
 usedDataInterfaces [
  &NiagaraDataInterfaceCurve
  &NiagaraDataInterfaceArrayColor
  &NiagaraDataInterfaceVector2DCurve
  &NiagaraDataInterfaceMeshRendererInfo
  &NiagaraDataInterfaceColorCurve
  &NiagaraDataInterfaceArrayFloat
 ]
 usedDynamicInputs [&MakeCustomFloatFromBool]
 usedModules [
  &self:SystemSpawnScript.SystemScriptSource.SystemScriptGraph.NiagaraNodeAssignment_0.SetVariables_#1
  &SystemState
 ]
}
emitters {
 Core_Flash001 {
  emitterUpdateScript [
   &EmitterState()
   &SpawnBurst_Instantaneous(Spawn Count=1, Use Spawn Probability=false) {enabled=false}
   &SpawnBurst_Instantaneous(Spawn Count=&RandomRangeInt(Minimum=12, Maximum=14), Spawn Group=1, Use Spawn Probability=false) {module=SpawnBurst_Instantaneous001}
   &SpawnBurst_Instantaneous(Spawn Count=5, Spawn Group=2, Use Spawn Probability=false, UseLoopCountLimit=false) {enabled=false, module=SpawnBurst_Instantaneous002}
   &SpawnBurst_Instantaneous(Spawn Group=3, Use Spawn Probability=false, UseLoopCountLimit=false) {module=SpawnBurst_Instantaneous003}
   &SpawnBurst_Instantaneous(Spawn Group=4, Use Spawn Probability=false, UseLoopCountLimit=false) {enabled=false, module=SpawnBurst_Instantaneous004}
   &SpawnBurst_Instantaneous(Spawn Group=5, Use Spawn Probability=false, UseLoopCountLimit=false) {enabled=false, module=SpawnBurst_Instantaneous005}
   &SpawnBurst_Instantaneous(Spawn Group=6, Use Spawn Probability=false, UseLoopCountLimit=false) {module=SpawnBurst_Instantaneous006}
  ]
  particleSpawnScript [
   /SetVariables(Particles.VisibilityTag=&SpawnGroup()) {isSetParameters=true}
   &PartitionParticles()
   &InitializeParticle(
    Color=&MergeHelper_ColorSwitch(
     0=<rgba 1 0.8404 0 1>
     1=<rgba 1 0.8404 0 1>
     2=<rgba 0 0 0 0>
     3=<rgba 1 0.8604 0.125 1>
     4=<rgba 1 1 1 1>
     5=<rgba 0 0 0 0>
     6=<rgba 2 1.6809 0 1>
     Index=&SpawnGroup()
     Num=::NewEnumerator5
    )
    Lifetime=&MergeHelper_FloatSwitch(
     0=0.22
     1=&RandomRangeFloat(Minimum=0.18, Maximum=0.22)
     2=&RandomRangeFloat(Minimum=0.2, Maximum=0.25)
     3=0.22
     4=0.25
     5=0.23
     6=0.25
     Index=&SpawnGroup()
     Num=::NewEnumerator5
    )
    Mesh Scale Mode=::NewEnumerator4
    Sprite Rotation Angle=&MergeHelper_FloatSwitch(
     0=0
     1=0
     2=0
     3=0
     4=0
     5=0
     6=&RandomRangeFloat(Minimum=0, Maximum=1)
     Index=&SpawnGroup()
     Num=::NewEnumerator5
    )
    Sprite Rotation Mode=::NewEnumerator1
    Sprite Size=&Multiply_Vector2D_ByFloat(
     Float=$System.Global Scale
     Vector2D=&MergeHelper_Vector2Switch(
      0=<0 0>
      1=&RandomRangeVector2D(Minimum=<8 65>, Maximum=<12 75>)
      2=&RandomRangeVector2D(Minimum=<5 40>, Maximum=<8 50>)
      3=&Vector2DFromFloat(Value=175)
      4=<1 1>
      5=&Vector2DFromFloat(Value=135)
      6=&Vector2DFromFloat(Value=80)
      Index=$Emitter.SpawnGroup
      Num=::NewEnumerator5
     )
    )
    Sprite Size Mode=::NewEnumerator4
    UsePositionOffset=false
    Write Material Random=true
   )
   &ShapeLocation(
    MaskSpawnGroup=false
    Rotation Angle:NiagaraFloat=&RandomRangeFloat(Minimum=0, Maximum=1)
    Rotation Mode=::NewEnumerator1
    Sphere Distribution=::NewEnumerator3
    Sphere Radius=&MergeHelper_FloatSwitch(
     0=0
     1=&Multiply_Float(A=0, B=$System.Global Scale)
     2=&Multiply_Float(A=27, B=$System.Global Scale)
     3=0
     4=0
     5=0
     Index=&SpawnGroup()
     Num=::NewEnumerator4
    )
    UseOwnerScale=true
   )
   &DynamicMaterialParameters(
    Index 0 Param 2=45
    Index 0 Param 3=1
    Index 0 Param 4=1
    Module.Param0WriteEnabled=false
    Module.Param1WriteEnabled=true
    Module.Param2WriteEnabled=true
    Module.Param3WriteEnabled=true
   ) {module=DynamicMaterialParameters001}
   &InitialMeshOrientation(
    Rotation=&RandomRangeVector(Minimum=<-1 -1 -1>, Maximum=<1 1 1>)
    Use Orientation Vector=true
    Use Rotation Vector=true
   ) {enabled=false}
   &CameraOffset(Camera Offset Amount=&MergeHelper_FloatSwitch(0=20, 1=2, 2=15, 3=25, 4=0, 5=25, 6=25, Index=&SpawnGroup(), Num=::NewEnumerator5))
   &AddVelocityFromPoint(
    Module.Use Falloff Distance=false
    Velocity Strength=&MergeHelper_FloatSwitch(
     0=0
     1=10
     2=&Multiply_Float(A=$System.Global Scale, B=&RandomRangeFloat(Minimum=300, Maximum=500))
     3=0
     4=0
     Index=&SpawnGroup()
     Num=::NewEnumerator3
    )
   )
   /SetVariables(Particles.Flare SUBUV:NiagaraFloat=&RandomRangeFloat(Minimum=0, Maximum=3)) {isSetParameters=true}
  ]
  particleUpdateScript [
   &ParticleState()
   &DynamicMaterialParameters(
    Index 0 Param 1=&FloatFromCurve(FloatCurve=DataInterface() {propertyValues {Curve {keys [<at 0.2 0 RCIM_Cubic>, {arriveTangent=3.6451, interpMode=RCIM_Cubic, leaveTangent=3.6451, tangentMode=RCTM_User, time=1, value=1}]}}})
    Module.Index0Float0Write=true
    Module.Index0Float1Write=true
    Module.Index0Float2Write=true
    Module.Index0Float3Write=true
   )
   &Drag(
    Drag=&MergeHelper_FloatSwitch(
     0=0
     1=0
     2=&FloatFromCurve(
      FloatCurve=DataInterface() {propertyValues {Curve {keys [
          {arriveTangent=1.9661, interpMode=RCIM_Cubic, leaveTangent=1.9661, leaveTangentWeight=67.9551, tangentMode=RCTM_User, time=0, value=0.1}
          {arriveTangent=0.1723, tangentMode=RCTM_User, time=1, value=1}
         ]}}}
      Scale Curve=20
     )
     3=0
     4=0
     5=0
     6=0
     Index=&SpawnGroup()
     Num=::NewEnumerator5
    )
    Use Linear Drag=true
    Use Rotational Drag=false
   )
   &ScaleColor(
    Scale Alpha=&MergeHelper_FloatSwitch(
     0=&FloatFromCurve(FloatCurve=DataInterface() {propertyValues {Curve {keys [<at 0 1>,<at 1 0>]}}})
     1=&FloatFromCurve(FloatCurve=DataInterface() {propertyValues {Curve {
         defaultValue=1
         keys [
          {arriveTangent=-0.592, interpMode=RCIM_Cubic, leaveTangent=-0.592, tangentMode=RCTM_User, time=0, value=1}
          {interpMode=RCIM_Cubic, tangentMode=RCTM_User, time=1, value=0}
         ]
        }}})
     2=1
     3=&FloatFromCurve(
      FloatCurve=DataInterface() {propertyValues {Curve {
         defaultValue=1
         keys [
          {interpMode=RCIM_Cubic, leaveTangentWeight=0.3461, tangentMode=RCTM_User, time=0.1, value=1}
          {arriveTangent=-1.2989, interpMode=RCIM_Cubic, leaveTangent=-1.2989, tangentMode=RCTM_User, time=1, value=0}
         ]
        }}}
      Scale Curve=&FN_FrontEndFloatModifier(
       Gameplay=&FloatFromCurve(
        FloatCurve=DataInterface() {propertyValues {Curve {keys [
            <at 0 0.65 RCIM_Cubic>
            {arriveTangent=0.875, interpMode=RCIM_Cubic, leaveTangent=0.875, time=0.1, value=0.7}
            {arriveTangentWeight=0.1329, interpMode=RCIM_Cubic, leaveTangentWeight=0.0759, tangentMode=RCTM_User, time=0.4, value=1}
            {arriveTangentWeight=0.0801, interpMode=RCIM_Cubic, leaveTangentWeight=0.1002, tangentMode=RCTM_User, time=0.6, value=1}
            {arriveTangent=-0.8778, interpMode=RCIM_Cubic, leaveTangent=-0.8778, time=0.9, value=0.7}
            <at 0.9987 0.65 RCIM_Cubic>
           ]}}}
        CurveIndex=Linked() {linkedVariable {name=NPC.FortniteNPC.TimeOfDay_0-1, type=NiagaraFloat}}
       )
       Front End=0.8
      )
     )
     4=1
     5=&FloatFromCurve(
      FloatCurve=DataInterface() {propertyValues {Curve {
         defaultValue=1
         keys [
          {interpMode=RCIM_Cubic, tangentMode=RCTM_User, time=0, value=1}
          {arriveTangent=-2.1802, interpMode=RCIM_Cubic, leaveTangent=-2.1802, tangentMode=RCTM_User, time=1, value=0}
         ]
        }}}
      Scale Curve=&FloatFromCurve(
       FloatCurve=DataInterface() {propertyValues {Curve {
          defaultValue=1
          keys [
           {arriveTangent=0.664, interpMode=RCIM_Cubic, leaveTangent=0.664, tangentMode=RCTM_User, time=0, value=0.45}
           {interpMode=RCIM_Cubic, tangentMode=RCTM_User, time=0.5, value=1}
           {tangentMode=RCTM_User, time=1, value=0.45}
          ]
         }}}
       CurveIndex=Linked() {linkedVariable {name=NPC.FortniteNPC.TimeOfDay_0-1, type=NiagaraFloat}}
       Scale Curve=1.2
      )
     )
     6=&FloatFromCurve(FloatCurve=DataInterface() {propertyValues {Curve {keys [<at 0.0045 1 RCIM_Cubic>, {arriveTangent=-1.344, interpMode=RCIM_Cubic, leaveTangent=-1.344, tangentMode=RCTM_User, time=1, value=0}]}}})
     Index=&SpawnGroup()
     Num=::NewEnumerator5
    )
    Scale RGB=&MergeHelper_VectorSwitch(
     0=&Multiply_VectorByFloat(
      Vector=&MakeVectorFromLinearColorRGB(Color=&ColorFromCurve(ColorCurve=DataInterface() {propertyValues {
          AlphaCurve {keys []}
          BlueCurve {keys [<at 0.032 2.8984 RCIM_Cubic>,<at 0.8557 0.334 RCIM_Cubic>]}
          GreenCurve {keys [<at 0.032 5.7139 RCIM_Cubic>,<at 0.8557 0.4939 RCIM_Cubic>]}
          RedCurve {keys [<at 0.032 7 RCIM_Cubic>,<at 0.8557 0.7995 RCIM_Cubic>]}
         }}))
      Float=&FloatFromCurve(
       FloatCurve=DataInterface() {propertyValues {Curve {keys [<at 0 4>,<at 0.2 1.5>,<at 1 1>]}}}
       Scale Curve=&FloatFromCurve(
        FloatCurve=DataInterface() {propertyValues {Curve {keys [
            <at 0.1 0.8 RCIM_Cubic>
            {arriveTangentWeight=0.1329, interpMode=RCIM_Cubic, leaveTangentWeight=0.0759, tangentMode=RCTM_User, time=0.35, value=1}
            {arriveTangentWeight=0.0801, interpMode=RCIM_Cubic, leaveTangentWeight=0.1002, tangentMode=RCTM_User, time=0.65, value=1}
            <at 0.9 0.8 RCIM_Cubic>
           ]}}}
        CurveIndex=Linked() {linkedVariable {name=NPC.FortniteNPC.TimeOfDay_0-1, type=NiagaraFloat}}
       )
      )
     )
     1=&Multiply_VectorByFloat(
      Vector=&MakeVectorFromLinearColorRGB(Color=&ColorFromCurve(ColorCurve=DataInterface() {propertyValues {
          AlphaCurve {keys []}
          BlueCurve {keys [<at 0 0 RCIM_Cubic>,<at 0.3535 0>,<at 0.9964 0>]}
          GreenCurve {keys [<at 0 2.2882 RCIM_Cubic>,<at 0.3535 0.8739>,<at 0.9964 0.275>]}
          RedCurve {keys [<at 0 3 RCIM_Cubic>,<at 0.3535 1>,<at 0.9964 0.849>]}
         }}))
      Float=&FloatFromCurve(
       FloatCurve=DataInterface() {propertyValues {Curve {keys [
           <at 0.1 0.8 RCIM_Cubic>
           {arriveTangentWeight=0.1329, interpMode=RCIM_Cubic, leaveTangentWeight=0.0759, tangentMode=RCTM_User, time=0.45, value=1}
           {arriveTangentWeight=0.0801, interpMode=RCIM_Cubic, leaveTangentWeight=0.1002, tangentMode=RCTM_User, time=0.55, value=1}
           <at 0.9 0.8 RCIM_Cubic>
          ]}}}
       CurveIndex=Linked() {linkedVariable {name=NPC.FortniteNPC.TimeOfDay_0-1, type=NiagaraFloat}}
      )
     )
     2=&MakeVectorFromLinearColorRGB(Color=$System.TOD Color)
     3=&Multiply_VectorByFloat(
      Vector=&MakeVectorFromLinearColorRGB(Color=&ColorFromCurve(ColorCurve=DataInterface() {propertyValues {
          AlphaCurve {keys []}
          BlueCurve {keys [<at 0.1614 0 RCIM_Cubic>,<at 0.6833 0>,<at 1 0.1352>]}
          GreenCurve {keys [<at 0.1614 2.2126 RCIM_Cubic>,<at 0.6833 0.6209>,<at 1 0.5509>]}
          RedCurve {keys [<at 0.1614 3 RCIM_Cubic>,<at 0.6833 1.0093>,<at 1 0.849>]}
         }}))
      Float=&FloatFromCurve(
       FloatCurve=DataInterface() {propertyValues {Curve {keys [
           <at 0.1 0.8 RCIM_Cubic>
           {arriveTangentWeight=0.1329, interpMode=RCIM_Cubic, leaveTangentWeight=0.0759, tangentMode=RCTM_User, time=0.363, value=1}
           {arriveTangentWeight=0.0801, interpMode=RCIM_Cubic, leaveTangentWeight=0.1002, tangentMode=RCTM_User, time=0.55, value=1}
           <at 0.9 0.8 RCIM_Cubic>
          ]}}}
       CurveIndex=Linked() {linkedVariable {name=NPC.FortniteNPC.TimeOfDay_0-1, type=NiagaraFloat}}
      )
     )
     4=<1 1 1>
     5=<1 1 1>
     6=&Multiply_VectorByFloat(
      Vector=&MakeVectorFromLinearColorRGB(Color=&ColorFromCurve(ColorCurve=DataInterface() {propertyValues {
          AlphaCurve {keys []}
          BlueCurve {keys [<at 0.1097 0 RCIM_Cubic>,<at 0.3451 0>,<at 0.9964 0>]}
          GreenCurve {keys [<at 0.1097 2.1915 RCIM_Cubic>,<at 0.3451 0.9637>,<at 0.9964 0.49>]}
          RedCurve {keys [<at 0.1097 3 RCIM_Cubic>,<at 0.3451 1.5249>,<at 0.9964 0.849>]}
         }}))
      Float=&FloatFromCurve(
       FloatCurve=DataInterface() {propertyValues {Curve {keys [<at 0 2>,<at 0.2 1.5>,<at 1 1>]}}}
       Scale Curve=&FloatFromCurve(
        FloatCurve=DataInterface() {propertyValues {Curve {keys [
            <at 0.1 0.8 RCIM_Cubic>
            {arriveTangentWeight=0.1329, interpMode=RCIM_Cubic, leaveTangentWeight=0.0759, tangentMode=RCTM_User, time=0.35, value=1}
            {arriveTangentWeight=0.0801, interpMode=RCIM_Cubic, leaveTangentWeight=0.1002, tangentMode=RCTM_User, time=0.65, value=1}
            <at 0.9 0.8 RCIM_Cubic>
           ]}}}
        CurveIndex=Linked() {linkedVariable {name=NPC.FortniteNPC.TimeOfDay_0-1, type=NiagaraFloat}}
       )
      )
     )
     Index=&SpawnGroup()
     Num=::NewEnumerator5
    )
    ScaleA=true
    ScaleRGB=true
   )
   &ScaleSpriteSize(
    Scale Factor=&MergeHelper_Vector2Switch(
     0=<0 0>
     1=&MakeVector2D(
      X=&FloatFromCurve(FloatCurve=DataInterface() {propertyValues {Curve {
          defaultValue=1
          keys [
           {interpMode=RCIM_Cubic, tangentMode=RCTM_User, time=0, value=0}
           {arriveTangentWeight=0.2236, interpMode=RCIM_Cubic, leaveTangentWeight=0.1491, tangentMode=RCTM_User, time=0.7, value=1}
           {arriveTangent=-2.2754, interpMode=RCIM_Cubic, leaveTangent=-2.2754, tangentMode=RCTM_User, time=1, value=0.5}
          ]
         }}})
      Y=&FloatFromCurve(FloatCurve=DataInterface() {propertyValues {Curve {
          defaultValue=1
          keys [
           {arriveTangent=3.143, interpMode=RCIM_Cubic, leaveTangent=3.143, tangentMode=RCTM_User, time=0, value=0}
           {arriveTangent=0.8537, interpMode=RCIM_Cubic, leaveTangent=0.8537, tangentMode=RCTM_User, time=1, value=1}
          ]
         }}})
     )
     2=&Multiply_Vector2D_ByFloat(
      Float=&FloatFromCurve(FloatCurve=DataInterface() {propertyValues {Curve {
          defaultValue=0
          keys [
           {arriveTangent=12.7226, interpMode=RCIM_Cubic, leaveTangent=12.7226, tangentMode=RCTM_User, time=0, value=0}
           {arriveTangentWeight=0.1123, interpMode=RCIM_Cubic, leaveTangentWeight=0.2599, tangentMode=RCTM_User, time=0.15, value=1}
           {arriveTangentWeight=0.3161, interpMode=RCIM_Cubic, leaveTangentWeight=0.4087, tangentMode=RCTM_User, time=0.4971, value=1}
           {arriveTangent=-3.108, interpMode=RCIM_Cubic, leaveTangent=-3.108, tangentMode=RCTM_User, time=1, value=0}
          ]
         }}})
      Vector2D=&Vector2DFromCurve(
       Vector2Curve=DataInterface() {propertyValues {
         XCurve {keys [
           {arriveTangent=8.4275, interpMode=RCIM_Cubic, leaveTangent=8.4275, tangentMode=RCTM_User, time=0, value=0.5844}
           {
            arriveTangent=0.1664
            arriveTangentWeight=0.0706
            interpMode=RCIM_Cubic
            leaveTangent=0.1664
            leaveTangentWeight=0.302
            tangentMode=RCTM_User
            time=0.1029
            value=1.0152
           }
           {arriveTangent=-1.3122, tangentMode=RCTM_User, time=1, value=0.5}
          ]}
         YCurve {keys [
           {arriveTangent=1.577, interpMode=RCIM_Cubic, leaveTangent=1.577, tangentMode=RCTM_User, time=-0.0025, value=0.9193}
           {arriveTangent=-1.4938, interpMode=RCIM_Cubic, leaveTangent=-1.4938, tangentMode=RCTM_User, time=0.2971, value=1.118}
           {arriveTangent=-0.2316, tangentMode=RCTM_User, time=1, value=0.5}
          ]}
        }}
       ScaleCurve=&Vector2DFromFloat(Value=$System.Global Scale)
      )
     )
     3=&Multiply_Vector2D_ByFloat(
      Float=&FloatFromCurve(FloatCurve=DataInterface() {propertyValues {Curve {
          defaultValue=0
          keys [
           {interpMode=RCIM_Cubic, tangentMode=RCTM_User, time=0, value=1}
           {arriveTangent=-1.4808, interpMode=RCIM_Cubic, leaveTangent=-1.4808, tangentMode=RCTM_User, time=1, value=0}
          ]
         }}})
      Vector2D=<1 1>
     )
     4=<0 0>
     5=&Vector2DFromFloat(Value=&FloatFromCurve(FloatCurve=DataInterface() {propertyValues {Curve {
          defaultValue=0
          keys [
           {arriveTangent=2.4894, interpMode=RCIM_Cubic, leaveTangent=2.4894, tangentMode=RCTM_User, time=0, value=0}
           {interpMode=RCIM_Cubic, tangentMode=RCTM_User, time=1, value=1}
          ]
         }}}))
     6=&Multiply_Vector2D_ByFloat(
      Float=&FloatFromCurve(FloatCurve=DataInterface() {propertyValues {Curve {
          defaultValue=0
          keys [
           {arriveTangent=2.0604, interpMode=RCIM_Cubic, leaveTangent=2.0604, tangentMode=RCTM_User, time=0, value=0.2}
           {arriveTangent=0.201, interpMode=RCIM_Cubic, leaveTangent=0.201, tangentMode=RCTM_User, time=1, value=1}
          ]
         }}})
      Vector2D=<1 1>
     )
     Index=&SpawnGroup()
     Num=::NewEnumerator5
    )
    Scale Sprite Size Mode=::NewEnumerator2
   )
   &SolveForcesAndVelocity(Clamp Velocity=false, Limit Acceleration=false)
   /SetVariables(Particles.Flare Dyn 1:Vector4f=&MakeVector4(
     X=&FloatFromCurve(FloatCurve=DataInterface() {propertyValues {Curve {keys [<at 0 1 RCIM_Cubic>,<at 0.8 0 RCIM_Cubic>]}}})
     Y=2
     Z=1
     W=1
    )) {isSetParameters=true}
   &FN_Light_Attributes(
    Color=&MultiplyLinearColor(
     Linear Color=<rgba 1 1 1 1>
     Scale Factor=&MakeVector4_FromLinearColor(LinearColor=<rgba 1 0.8556 0.1287 1>)
    )
    Create Color Attribute=true
    Create Exponent Attribute=true
    Create Radius Attribute=true
    Exponent=3
    Radius=95
    Scale Color=&FloatFromCurve(
     FloatCurve=DataInterface() {propertyValues {Curve {
        defaultValue=1
        keys [
         {interpMode=RCIM_Cubic, tangentMode=RCTM_User, time=0, value=1}
         {arriveTangent=-0.745, interpMode=RCIM_Cubic, leaveTangent=-0.745, tangentMode=RCTM_User, time=1, value=0}
        ]
       }}}
     Scale Curve=15
    )
   )
  ]
  rendererClasses [
   &NiagaraMeshRendererProperties
   &NiagaraSpriteRendererProperties
   &NiagaraLightRendererProperties
  ]
  renderers [
   {rendererClass=&NiagaraMeshRendererProperties, rendererIndex=0}
   {rendererClass=&NiagaraSpriteRendererProperties, rendererIndex=1}
   {rendererClass=&NiagaraSpriteRendererProperties, rendererIndex=2}
   {rendererClass=&NiagaraSpriteRendererProperties, rendererIndex=3}
   {rendererClass=&NiagaraLightRendererProperties, rendererIndex=4}
   {rendererClass=&NiagaraSpriteRendererProperties, rendererIndex=5}
   {rendererClass=&NiagaraSpriteRendererProperties, rendererIndex=6}
  ]
 }
 FN_PhysMatDebris {
  emitterUpdateScript [
   &EmitterState()
   &SpawnBurst_Instantaneous(
    Spawn Count=&RandomRangeInt(Minimum=5, Maximum=10)
    Use Spawn Probability=false
    UseLoopCountLimit=false
   )
  ]
  particleSpawnScript [
   /SetVariables(Particles.BiasedRandom=&FloatFromCurve(
     FloatCurve=DataInterface() {propertyValues {Curve {keys [
         <at 0.2 1 RCIM_Cubic>
         {arriveTangent=-3.3394, interpMode=RCIM_Cubic, leaveTangent=-3.3394, tangentMode=RCTM_User, time=0.5, value=0.5}
         <at 0.8 0 RCIM_Cubic>
        ]}}}
     CurveIndex=&RandomRangeFloat(Minimum=0, Maximum=1)
    )) {isSetParameters=true}
   &InitializeParticle(
    Color Mode=::NewEnumerator0
    Lifetime=&RandomRangeFloat(Minimum=0.5, Maximum=1.25)
    Mass=&Lerp_Float(A=0.6, B=1.2, Alpha=$Particles.BiasedRandom)
    Mass Mode=::NewEnumerator1
    Mesh Scale=&Multiply_VectorByFloat(
     Vector=&RandomRangeVector(Minimum=<0.9 0.4 0.5>, Maximum=<1 1 1>)
     Float=&Lerp_Float(A=0.7, B=1, Alpha=$Particles.BiasedRandom)
    )
    Mesh Scale Mode=::NewEnumerator4
    UsePositionOffset=false
    Write Material Random=true
   )
   &FN_PhysicalSurface(
    Dirt Mesh Count=2
    Glass Mesh Count=3
    Ice Mesh Count=2
    Metal Gravity=&RandomRangeFloat(Minimum=-350, Maximum=-1800)
    Metal Mesh Count=2
    Metal Velocity=&RandomRangeFloat(Minimum=80, Maximum=375)
    Metal Velocity Cone Angle=130
    Physical Surface=$User.PhysicalSurface
    Stone Gravity=&RandomRangeFloat(Minimum=-700, Maximum=-1250)
    Stone Mesh Count=2
    Stone Velocity=&RandomRangeFloat(Minimum=80, Maximum=550)
    Stone Velocity Cone Angle=120
    Wood Gravity=&RandomRangeFloat(Minimum=-450, Maximum=-900)
    Wood Mesh Count=2
    Wood Scale=1.5
    Wood Velocity=&RandomRangeFloat(Minimum=80, Maximum=600)
    Wood Velocity Cone Angle=130
   )
   &AddVelocityInCone(
    Cone Angle=$Output.FN_PhysicalSurface.SurfaceResponseConeAngle
    Module.Use Velocity Falloff On Cone Axis=false
    Velocity Strength=$Output.FN_PhysicalSurface.SurfaceResponseVelocity
   )
   &InitialMeshOrientation(Use Rotation Vector=false)
   &AddRotationalVelocity(Rotation Rate=&RandomVector(Vector Scale=&RandomRangeFloat(Minimum=1, Maximum=2)))
   &ApplyInitialForces()
   /SetVariables(Particles.GravityOverrideValue=$Output.FN_PhysicalSurface.GravityOverride) {isSetParameters=true}
  ]
  particleUpdateScript [
   &ParticleState()
   &GravityForce(Gravity=$Particles.GravityOverrideValue)
   &Collision(
    Advanced Aging Rate=1
    Age Colliding Particles=false
    CPU Collision Trace Channel=ECC_WorldStatic
    Collision Enabled=&Boolean_AndOperation(
     Bool1=&SetBoolByFloatComparison(A=$Particles.Initial.Lifetime, B=1)
     Bool2=&SetBoolByFloatComparison(A=$Particles.Age, B=0.12)
    )
    Control Roll On Collision=true
    Friction=1
    Friction During a Bounce=0.25
    Max CPU Trace Length=100
    Module.Control Rest State Via Surface Normal=false
    Module.EnableMaxCollisionCount=false
    Module.Randomize Collision Normal Vector=true
    Randomize Collision Normal=0.25
    ScaleRateOfRotation=false
   ) {enabled=false}
   &SolveForcesAndVelocity(Clamp Velocity=false, Limit Acceleration=false)
   &ScaleMeshSize(Scale Factor=&VectorFromFloat(Value=&FloatFromCurve(FloatCurve=DataInterface() {propertyValues {Curve {keys [<at 0.6 1 RCIM_Cubic>, {arriveTangent=-6.308, interpMode=RCIM_Cubic, leaveTangent=-6.308, tangentMode=RCTM_User, time=1, value=0}]}}})))
   &DynamicMaterialParameters(Module.Param0WriteEnabled=true, Module.Param1WriteEnabled=true, Module.Param2WriteEnabled=false, Module.Param3WriteEnabled=false)
  ]
  rendererClasses [&NiagaraSpriteRendererProperties, &NiagaraMeshRendererProperties]
  renderers [
   {rendererClass=&NiagaraSpriteRendererProperties}
   {rendererClass=&NiagaraMeshRendererProperties, rendererIndex=1}
   {rendererClass=&NiagaraMeshRendererProperties, rendererIndex=2}
   {rendererClass=&NiagaraMeshRendererProperties, rendererIndex=3}
   {rendererClass=&NiagaraMeshRendererProperties, rendererIndex=4}
   {rendererClass=&NiagaraMeshRendererProperties, rendererIndex=5}
   {rendererClass=&NiagaraMeshRendererProperties, rendererIndex=6}
   {rendererClass=&NiagaraMeshRendererProperties, rendererIndex=7}
   {rendererClass=&NiagaraMeshRendererProperties, rendererIndex=8}
  ]
 }
 Hearts {
  emitterUpdateScript [
   &EmitterState(Life Cycle Mode=::NewEnumerator1, Loop Behavior=::NewEnumerator1, Loop Delay=0.005, Loop Duration=0.25, UseLoopDelay=false)
   &SpawnBurst_Instantaneous(Spawn Count=3, Use Spawn Probability=false, UseLoopCountLimit=false)
  ]
  particleSpawnScript [
   &FN_SetSpawnGroup(Number of Spawn Groups=2)
   &InitializeParticle(
    Color=&SelectLinearColorFromArray(
     Array Sampling Mode=::NewEnumerator1
     Direct Array Index=&ReturnExecIndex()
     Direct Array Mode=::NewEnumerator1
     Color Selection Array=DataInterface() {propertyValues {ColorData [<rgba 0.0766 0.5234 0.0545 1>, <rgba 0.0409 0.2232 0.0319 1>]}}
     Random Seed=?
    )
    Lifetime=&FloatFromCurve(
     FloatCurve=DataInterface() {propertyValues {Curve {keys [<at 0 0.3>,<at 1 0.5>]}}}
     CurveIndex=&ReturnNormalizedExecIndex()
    )
    Sprite Size Mode=::NewEnumerator3
    Uniform Sprite Size=&Multiply_Float(A=&RandomRangeFloat(Minimum=13, Maximum=18), B=1)
    UsePositionOffset=false
    Write Material Random=true
   )
   &ShapeLocation(
    MaskSpawnGroup=false
    Override Seed=false
    Sphere Radius=&Multiply_Float(A=15, B=$System.Global Scale)
    Sphere Surface Distribution=1
    UseOwnerScale=true
   ) {enabled=false}
   &ShapeLocation(
    Handle Radius=1
    Large Radius=&Multiply_Float(A=7, B=$System.Global Scale)
    MaskSpawnGroup=false
    Rotation Angle:NiagaraFloat=&RandomRangeFloat(Minimum=0, Maximum=0.5)
    Rotation Axis=&FN_FrontEndVector3Modifier(Gameplay=<1 1 0>, Front End=<1 0 0>)
    Rotation Mode=::NewEnumerator1
    Shape Primitive=::NewEnumerator3
    Torus Distribution Mode:ENiagaraTorusDistributionMode=::NewEnumerator1
    U Position=&ReturnNormalizedExecIndex(Normalized Index Includes One=false)
    UseOwnerScale=true
   ) {module=ShapeLocation001}
   &AddVelocity(
    Module.Use Falloff Distance=false
    Override Seed=false
    Velocity Mode=::NewEnumerator1
    Velocity Speed=&Multiply_Float(
     A=&FloatFromCurve(
      FloatCurve=DataInterface() {propertyValues {Curve {keys [<at 0 0.6>,<at 1 1>]}}}
      CurveIndex=&ReturnNormalizedExecIndex()
      Scale Curve=275
     )
     B=$System.Global Scale
    )
   )
   &CameraOffset(Camera Offset Amount=&FN_FrontEndFloatModifier(Gameplay=25, Front End=5))
  ]
  particleUpdateScript [
   &ParticleState()
   &SpriteRotationRate(Rotation Rate=&RandomRangeFloat(Minimum=-45, Maximum=45))
   &Drag(
    Drag=&FN_FrontEndFloatModifier(Gameplay=9, Front End=7)
    Use Linear Drag=true
    Use Rotational Drag=true
   )
   &SolveForcesAndVelocity(Clamp Velocity=false, Limit Acceleration=false)
   &ScaleSpriteSize(
    Non-Uniform Curve Scale=&Multiply_Vector2D_ByFloat(
     Float=$System.Global Scale
     Vector2D=&Vector2DFromFloat(Value=&FloatFromCurve(FloatCurve=DataInterface() {propertyValues {Curve {
          defaultValue=1
          keys [
           {arriveTangent=3.1796, interpMode=RCIM_Cubic, leaveTangent=3.1796, tangentMode=RCTM_User, time=0, value=0}
           {interpMode=RCIM_Cubic, tangentMode=RCTM_User, time=0.1833, value=1}
           {arriveTangent=-3.7394, tangentMode=RCTM_User, time=1, value=0}
          ]
         }}}))
    )
    Non-Uniform Curve Sprite Scale=DataInterface() {propertyValues {
      XCurve {keys [
        <at 0 0.6029 RCIM_Cubic>
        {arriveTangentWeight=0.0864, interpMode=RCIM_Cubic, leaveTangentWeight=0.1235, tangentMode=RCTM_User, time=0.2782, value=1.3837}
        {arriveTangentWeight=0.127, interpMode=RCIM_Cubic, leaveTangentWeight=0.0792, tangentMode=RCTM_User, time=0.5872, value=0.7698}
        {arriveTangentWeight=0.0712, interpMode=RCIM_Cubic, leaveTangentWeight=0.0686, tangentMode=RCTM_User, time=0.7971, value=1.1757}
        {arriveTangentWeight=0.0674, interpMode=RCIM_Cubic, tangentMode=RCTM_User, time=0.9991, value=0.8488}
       ]}
      YCurve {keys [
        <at 0 1.3741 RCIM_Cubic>
        {arriveTangentWeight=0.0853, interpMode=RCIM_Cubic, leaveTangentWeight=0.1219, tangentMode=RCTM_User, time=0.2782, value=0.6394}
        {arriveTangentWeight=0.1191, interpMode=RCIM_Cubic, leaveTangentWeight=0.0743, tangentMode=RCTM_User, time=0.5872, value=1.2415}
        {arriveTangentWeight=0.0748, interpMode=RCIM_Cubic, leaveTangentWeight=0.0684, tangentMode=RCTM_User, time=0.7983, value=0.8301}
        {arriveTangentWeight=0.0644, interpMode=RCIM_Cubic, tangentMode=RCTM_User, time=0.9916, value=1.0963}
       ]}
     }}
    Scale Sprite Size Mode=::NewEnumerator3
   )
   &ScaleColor(
    Scale Alpha=&FloatFromCurve(
     FloatCurve=DataInterface() {propertyValues {Curve {
        defaultValue=0
        keys [
         {arriveTangent=13.0585, interpMode=RCIM_Cubic, leaveTangent=13.0585, tangentMode=RCTM_User, time=0, value=0}
         {arriveTangent=0.0515, interpMode=RCIM_Cubic, leaveTangent=0.0515, tangentMode=RCTM_User, time=0.15, value=1}
         {arriveTangent=-1.8704, interpMode=RCIM_Cubic, leaveTangent=-1.8704, tangentMode=RCTM_User, time=1, value=0}
        ]
       }}}
     Scale Curve=10
    )
    Scale RGB=&Multiply_VectorByFloat(
     Vector=&MakeVectorFromLinearColorRGB(Color=$System.TOD Color)
     Float=&FloatFromCurve(FloatCurve=DataInterface() {propertyValues {Curve {keys [<at 0 3>,<at 0.3 1>,<at 1 1>]}}})
    )
    ScaleA=true
    ScaleRGB=true
   ) {module=ScaleColor001}
  ]
  rendererClasses [&NiagaraSpriteRendererProperties]
  renderers [{rendererClass=&NiagaraSpriteRendererProperties}]
 }
 Smoke {
  emitterUpdateScript [
   &EmitterState()
   &SpawnBurst_Instantaneous(
    Spawn Count=&RandomRangeInt(Minimum=2, Maximum=3)
    Use Spawn Probability=false
    UseLoopCountLimit=false
   )
  ]
  particleSpawnScript [
   &InitializeParticle(
    Color Maximum=<rgba 0.7811 0.7626 0.7982 0.2>
    Color Minimum=<rgba 0.5646 0.6299 0.7617 0.15>
    Color Mode=::NewEnumerator2
    Lifetime Max=0.45
    Lifetime Min=0.35
    Lifetime Mode=::NewEnumerator1
    Sprite Size Mode=::NewEnumerator1
    Uniform Sprite Size Max=150
    Uniform Sprite Size Min=100
    UsePositionOffset=false
    Write Material Random=true
   )
   &CameraOffset(Camera Offset Amount=&RandomRangeFloat(Minimum=15, Maximum=20, Evaluation Type=::NewEnumerator1))
   &ShapeLocation(
    MaskSpawnGroup=false
    Override Seed=false
    Sphere Radius=&Multiply_Float(A=10, B=$System.Global Scale)
    UseOwnerScale=true
   )
   &AddVelocity(
    Velocity=&Multiply_VectorByFloat(Vector=$Particles.ShapeLocation.ShapeVector, Float=10)
    Velocity Speed Scale=$System.Global Scale
   )
  ]
  particleUpdateScript [
   &ParticleState()
   &ScaleSpriteSize(
    Uniform Curve Scale=&FN_FrontEndFloatModifier(Gameplay=1, Front End=$System.Global Scale)
    Uniform Curve Sprite Scale=DataInterface() {propertyValues {Curve {keys [
        {arriveTangent=3.5075, interpMode=RCIM_Cubic, leaveTangent=3.5075, tangentMode=RCTM_User, time=0, value=0}
        {arriveTangent=1, interpMode=RCIM_Cubic, leaveTangent=1, time=0.2433, value=0.6441}
        {interpMode=RCIM_Cubic, tangentMode=RCTM_User, time=1, value=1}
       ]}}}
   )
   &ScaleColor(
    Scale Alpha=&FloatFromCurve(FloatCurve=DataInterface() {propertyValues {Curve {keys [
         {arriveTangent=1.5855, interpMode=RCIM_Cubic, leaveTangent=1.5855, tangentMode=RCTM_User, time=0, value=0}
         {arriveTangent=-0.038, interpMode=RCIM_Cubic, leaveTangent=-0.038, tangentMode=RCTM_User, time=0.25, value=1}
         {arriveTangent=-0.0405, interpMode=RCIM_Cubic, leaveTangent=-0.0405, tangentMode=RCTM_User, time=1, value=0}
        ]}}})
    Scale RGB=&MakeVectorFromLinearColorRGB(Color=$System.TOD Color)
    ScaleA=true
    ScaleRGB=true
   )
   &DynamicMaterialParameters(
    Index 0 Param 1=&Add_Float(
     A=&RandomRangeFloat(Minimum=0, Maximum=1)
     B=&FloatFromCurve(
      FloatCurve=DataInterface() {propertyValues {Curve {keys [
          {arriveTangent=0.9009, interpMode=RCIM_Cubic, leaveTangent=0.9009, tangentMode=RCTM_User, time=0, value=0}
          {arriveTangent=1.5947, interpMode=RCIM_Cubic, leaveTangent=1.5947, tangentMode=RCTM_User, time=0.2557, value=0.5257}
          <at 1 1 RCIM_Cubic>
         ]}}}
      CurveIndex=$Particles.Age
      Scale Curve=0.25
     )
    )
    Module.Index0Float0Write=true
    Module.Index0Float1Write=true
    Module.Index0Float2Write=true
    Module.Index0Float3Write=true
   ) {module=DynamicMaterialParameters001}
   &Drag(Drag=3, Use Linear Drag=true, Use Rotational Drag=true)
   &SolveForcesAndVelocity(Clamp Velocity=false, Limit Acceleration=false)
  ]
  rendererClasses [&NiagaraSpriteRendererProperties]
  renderers [
   {rendererClass=&NiagaraSpriteRendererProperties}
   {rendererClass=&NiagaraSpriteRendererProperties, rendererIndex=1}
  ]
 }
 Sparks {
  emitterUpdateScript [
   &EmitterState()
   &SpawnBurst_Instantaneous(Spawn Count=&UniformRangedInt(Minimum=30, Maximum=35), Use Spawn Probability=false)
   &SpawnRate(
    Spawn Probability=&FloatFromCurve(
     FloatCurve=DataInterface() {propertyValues {Curve {keys [{arriveTangent=-12.2872, interpMode=RCIM_Cubic, leaveTangent=-12.2872, tangentMode=RCTM_User, time=0, value=1}, <at 0.1 0>]}}}
     CurveIndex=$System.Age
    )
    SpawnRate=111
    Use Spawn Probability=true
   )
  ]
  particleSpawnScript [
   &InitializeParticle(
    Color=<rgba 1 0.5971 0.0008 1>
    Lifetime Max=0.45
    Lifetime Min=0.166
    Lifetime Mode=::NewEnumerator1
    Sprite Size Mode=::NewEnumerator3
    Uniform Sprite Size=&Multiply_Float(A=$System.Global Scale, B=&RandomRangeFloat(Minimum=3, Maximum=9))
    UsePositionOffset=false
    Write Material Random=true
   )
   &ShapeLocation(
    MaskSpawnGroup=false
    Override Seed=false
    Sphere Radius=&Multiply_Float(A=9, B=$System.Global Scale)
    Sphere Surface Distribution=1
    UseOwnerScale=true
   )
   &AddVelocity(
    Module.Use Falloff Distance=false
    Override Seed=false
    Velocity Mode=::NewEnumerator1
    Velocity Speed=&Multiply_Float(A=&RandomRangeFloat(Minimum=222, Maximum=666), B=$System.Global Scale)
   )
   &DynamicMaterialParameters(
    Index 0 Param 1=3
    Index 0 Param 2=20
    Index 0 Param 3=1
    Module.Param0WriteEnabled=true
    Module.Param1WriteEnabled=true
    Module.Param2WriteEnabled=true
    Module.Param3WriteEnabled=true
   )
  ]
  particleUpdateScript [
   &ParticleState()
   &GravityForce(Gravity=<0 0 -666>)
   &Drag(Drag=&UniformRangedFloat(Minimum=2, Maximum=6), Use Linear Drag=true, Use Rotational Drag=false)
   &SolveForcesAndVelocity(Clamp Velocity=false, Limit Acceleration=false)
   &ScaleSpriteSize(Scale Factor=&Vector2DFromCurve(
     Vector2Curve=DataInterface() {propertyValues {
       XCurve {keys [
         {arriveTangent=9.4237, interpMode=RCIM_Cubic, leaveTangent=9.4237, tangentMode=RCTM_User, time=0, value=0.1}
         {arriveTangentWeight=0.0603, interpMode=RCIM_Cubic, leaveTangentWeight=0.2747, tangentMode=RCTM_User, time=0.18, value=0.8}
         {arriveTangent=-0.8857, interpMode=RCIM_Cubic, leaveTangent=-0.8857, tangentMode=RCTM_User, time=1, value=0}
        ]}
       YCurve {keys [
         {arriveTangent=10.1142, interpMode=RCIM_Cubic, leaveTangent=10.1142, tangentMode=RCTM_User, time=0, value=0.1}
         {arriveTangentWeight=0.0603, interpMode=RCIM_Cubic, leaveTangentWeight=0.2743, tangentMode=RCTM_User, time=0.18, value=0.8}
         {arriveTangent=-0.8857, interpMode=RCIM_Cubic, leaveTangent=-0.8857, tangentMode=RCTM_User, time=0.9988, value=0}
        ]}
      }}
     ScaleCurve=<1 1>
    ))
   &ScaleSpriteSizeBySpeed(Max Scale Factor=<0.5 4>, Min Scale Factor=<0.8 1>, Velocity Threshold=666)
   &ScaleColor(Scale RGB=&VectorFromFloat(Value=2), ScaleA=true, ScaleRGB=true) {module=ScaleColor001}
  ]
  rendererClasses [&NiagaraSpriteRendererProperties]
  renderers [{rendererClass=&NiagaraSpriteRendererProperties}]
 }
 Sparks001 {
  emitterUpdateScript [
   &EmitterState(
    Enable Distance Culling=true
    Max Distance Response=::KillImmediately
    Module.bApplySpawnCountScaleByDistance=true
    Module.bUseMaxDistance=true
    Module.bUseMinDistance=false
    Scalability Mode=::NewEnumerator1
    Scale Spawn Count=true
    Spawn Count Scale By Distance=DataInterface() {propertyValues {Curve {keys [<at 0.3676 0.9657>,<at 0.9962 0.4982>]}}}
   )
   &SpawnBurst_Instantaneous(Spawn Count=14, Use Spawn Probability=false, UseLoopCountLimit=true)
  ]
  particleSpawnScript [
   &PartitionParticles()
   &InitializeParticle(
    Color=&SelectLinearColorFromArray(
     Array Sampling Mode=::NewEnumerator1
     Direct Array Index=$Particles.PartitionParticles.Partition
     Color Selection Array=DataInterface() {propertyValues {ColorData [<rgba 0 0.3307 0.0445 1>, <rgba 1.5 1.1622 0 1>]}}
     Random Seed=?
    )
    Lifetime=&FloatFromCurve(
     FloatCurve=DataInterface() {propertyValues {Curve {keys [<at 0.15 0.65>, {arriveTangent=-0.9801, time=0.4, value=0.45}, {arriveTangent=-0.2795, leaveTangent=-0.2795, time=0.8, value=0.35}]}}}
     CurveIndex=&RandomRangeFloat(Minimum=0, Maximum=1)
    )
    Sprite Size Mode=::NewEnumerator3
    Uniform Sprite Size=&Multiply_Float(
     A=&RandomRangeFloat(Minimum=5.5, Maximum=6)
     B=&MaskFloat(
      Mask Source=1
      Mask Parameter=$Particles.PartitionParticles.Partition
      Mask=2
      Masked Float=0.85
      Passthrough Float=&RandomRangeFloat(Minimum=1.2, Maximum=1)
     )
    )
    UsePositionOffset=false
    Write Material Random=true
   )
   &SphereLocation(
    MaskSpawnGroup=false
    Override Seed=false
    Sphere Radius=5
    Surface Only=false
    UseNonUniformScale=false
    UseOffset=false
    UseSphereOrientation=false
   ) {module=SphereLocation001}
   &AddVelocityFromPoint(
    Module.Use Falloff Distance=false
    Velocity Strength=&Multiply_Float(
     A=&RandomRangeFloat(Minimum=500, Maximum=700)
     B=&Multiply_Float(
      A=&SelectFloatFromArray(Direct Array Index=$Particles.PartitionParticles.Partition, Float Selection Array=DataInterface() {propertyValues {FloatData [1,0.9,0.55]}})
      B=$System.Global Scale
     )
    )
   )
  ]
  particleUpdateScript [
   &ParticleState()
   &GravityForce(Gravity=&Multiply_VectorByFloat(
     Vector=<0 0 -1>
     Float=&FloatFromCurve(
      FloatCurve=DataInterface() {propertyValues {Curve {
         defaultValue=1
         keys [
          {interpMode=RCIM_Cubic, leaveTangentWeight=0.4106, tangentMode=RCTM_User, time=0, value=0.5}
          {arriveTangent=0.9376, interpMode=RCIM_Cubic, leaveTangent=0.9376, tangentMode=RCTM_User, time=1, value=1}
         ]
        }}}
      Scale Curve=450
     )
    ))
   &Drag(Drag=&UniformRangedFloat(Minimum=4, Maximum=9), Use Linear Drag=true, Use Rotational Drag=false)
   &SolveForcesAndVelocity(Clamp Velocity=false, Limit Acceleration=false)
   &ScaleSpriteSize(Scale Factor=&Vector2DFromCurve(
     Vector2Curve=DataInterface() {propertyValues {
       XCurve {keys [
         {arriveTangent=5.9728, interpMode=RCIM_Cubic, leaveTangent=5.9728, tangentMode=RCTM_User, time=0, value=0.1}
         {arriveTangent=-0.1, interpMode=RCIM_Cubic, leaveTangent=-0.1, tangentMode=RCTM_User, time=0.18, value=0.8}
         {interpMode=RCIM_Cubic, tangentMode=RCTM_User, time=1, value=0}
        ]}
       YCurve {keys [
         {arriveTangent=6.2715, interpMode=RCIM_Cubic, leaveTangent=6.2715, tangentMode=RCTM_User, time=0, value=0.1}
         {arriveTangent=-0.1001, interpMode=RCIM_Cubic, leaveTangent=-0.1001, tangentMode=RCTM_User, time=0.18, value=0.8}
         {interpMode=RCIM_Cubic, tangentMode=RCTM_User, time=0.9988, value=0}
        ]}
      }}
     ScaleCurve=&Multiply_Vector2D_ByFloat(Float=$System.Global Scale, Vector2D=<1 1>)
    ))
   &ScaleSpriteSizeBySpeed(Max Scale Factor=<0.5 8>, Min Scale Factor=<0.8 1.5>)
   &ScaleColor(
    Scale Alpha=&FloatFromCurve(FloatCurve=DataInterface() {propertyValues {Curve {keys [<at 0.4 1 RCIM_Cubic>,<at 1 0>]}}})
    Scale RGB=&Multiply_Vector(
     A=&MakeVectorFromLinearColorRGB(Color=$System.TOD Color)
     B=&MakeVectorFromLinearColorRGB(Color=&ColorFromCurve(ColorCurve=DataInterface() {propertyValues {
         AlphaCurve {keys []}
         BlueCurve {keys [<at 0.1823 2>,<at 1 1>]}
         GreenCurve {keys [<at 0.1823 2>,<at 1 1>]}
         RedCurve {keys [<at 0.1823 2>,<at 1 1>]}
        }}))
    )
    ScaleA=true
    ScaleRGB=true
   )
  ]
  rendererClasses [&NiagaraSpriteRendererProperties]
  renderers [{rendererClass=&NiagaraSpriteRendererProperties}]
 }
 Spikes001 {
  emitterUpdateScript [
   &EmitterState()
   &SpawnBurst_Instantaneous(Spawn Count=9, Use Spawn Probability=true, UseLoopCountLimit=false)
  ]
  particleSpawnScript [
   &PartitionParticles()
   &InitializeParticle(
    Color=&SelectLinearColorFromArray(
     Array Sampling Mode=::NewEnumerator1
     Direct Array Index=$Particles.PartitionParticles.Partition
     Color Selection Array=DataInterface() {propertyValues {ColorData [<rgba 0.0623 0.0387 1 1>, <rgba 0.0849 0.154 0.8932 1>]}}
     Random Seed=?
    )
    Lifetime=0.17
    Sprite Rotation Mode=::NewEnumerator2
    Sprite Size=&Multiply_Vector2D_ByFloat(
     Float=$System.Global Scale
     Vector2D=&Multiply_Vector2D_ByFloat(Float=&RandomRangeFloat(Minimum=0.85, Maximum=1), Vector2D=<40 70>)
    )
    Sprite Size Mode=::NewEnumerator4
    UsePositionOffset=false
    Write Material Random=true
   )
   &ShapeLocation(
    Handle Radius=0
    Large Radius=&Multiply_Float(A=1, B=$System.Global Scale)
    MaskSpawnGroup=false
    Offset=<-1 0 0>
    Offset Coordinate Space=::Local
    Offset Mode=::NewEnumerator0
    Rotation Axis=<0.5 1 0>
    Rotation Mode=::NewEnumerator1
    Shape Primitive=::NewEnumerator3
    Torus Distribution Mode:ENiagaraTorusDistributionMode=::NewEnumerator1
    U Position=&Add_Float(
     A=&ReturnNormalizedExecIndex(Normalized Index Includes One=false)
     B=&RandomRangeFloat(Minimum=-0.1, Maximum=0.1)
    )
    UseOwnerScale=true
   ) {enabled=false}
   &ShapeLocation(
    MaskSpawnGroup=false
    Offset Mode=::NewEnumerator0
    Rotation Angle:NiagaraFloat=&RandomRangeFloat(Minimum=0, Maximum=1)
    Rotation Axis=<0 0 1>
    Rotation Mode=::NewEnumerator1
    Sphere Distribution=::NewEnumerator3
    Sphere Radius=&Multiply_Float(A=1, B=$System.Global Scale)
    UseOwnerScale=true
   ) {module=ShapeLocation001}
   &AddVelocity(
    Module.Use Falloff Distance=false
    Override Seed=false
    Velocity Mode=::NewEnumerator1
    Velocity Speed=&Multiply_Float(A=&RandomRangeFloat(Minimum=17, Maximum=52.5), B=$System.Global Scale)
   )
   &DynamicMaterialParameters(
    Index 0 Param 1=3
    Index 0 Param 2=20
    Index 0 Param 3=1
    Module.Param0WriteEnabled=true
    Module.Param1WriteEnabled=true
    Module.Param2WriteEnabled=true
    Module.Param3WriteEnabled=true
   )
   &CameraOffset(Camera Offset Amount=15)
   /SetVariables(Particles.VisibilityTag=$Particles.PartitionParticles.Partition) {isSetParameters=true}
  ]
  particleUpdateScript [
   &ParticleState()
   &ScaleSpriteSize(
    Scale Factor=&MakeVector2D(
     X=&FloatFromCurve(FloatCurve=DataInterface() {propertyValues {Curve {
         defaultValue=1
         keys [{interpMode=RCIM_Cubic, tangentMode=RCTM_User, time=0.5, value=1}, {interpMode=RCIM_Cubic, tangentMode=RCTM_User, time=1, value=0}]
        }}})
     Y=1
    )
    Scale Sprite Size Mode=::NewEnumerator2
   ) {module=ScaleSpriteSize001}
   &ScaleSpriteSize(Uniform Curve Sprite Scale=DataInterface() {propertyValues {Curve {
       defaultValue=0
       keys [
        {arriveTangent=2.4124, interpMode=RCIM_Cubic, leaveTangent=2.4124, tangentMode=RCTM_User, time=0, value=0}
        {arriveTangent=1.1179, interpMode=RCIM_Cubic, leaveTangent=1.1179, tangentMode=RCTM_User, time=0.2859, value=0.5411}
        {arriveTangent=0.0224, interpMode=RCIM_Cubic, leaveTangent=0.0224, tangentMode=RCTM_User, time=1, value=1}
       ]
      }}})
   &SolveForcesAndVelocity(Clamp Velocity=false, Limit Acceleration=false)
   &DynamicMaterialParameters(
    Module.Index1Float0Write=false
    Module.Index1Float1Write=false
    Module.Index1Float2Write=true
    Module.Index1Float3Write=false
    Write Parameter Index 0=::NewEnumerator0
    Write Parameter Index 1=::NewEnumerator1
   ) {enabled=false, module=DynamicMaterialParameters001}
   &ScaleColor(
    Scale Alpha=&FloatFromCurve(FloatCurve=DataInterface() {propertyValues {Curve {keys [
         {arriveTangent=19.5598, interpMode=RCIM_Cubic, leaveTangent=19.5598, tangentMode=RCTM_User, time=0, value=0}
         {arriveTangentWeight=0.0557, interpMode=RCIM_Cubic, leaveTangentWeight=0.3721, tangentMode=RCTM_User, time=0.1002, value=1}
         {arriveTangentWeight=0.3481, interpMode=RCIM_Cubic, leaveTangentWeight=0.1001, tangentMode=RCTM_User, time=0.5, value=1}
         <at 1.0011 0.002 RCIM_Cubic>
        ]}}})
    Scale RGB=&Multiply_VectorByFloat(
     Vector=&Multiply_Vector(
      A=&MakeVectorFromLinearColorRGB(Color=$System.TOD Color)
      B=&MakeVectorFromLinearColorRGB(Color=&ColorFromCurve(ColorCurve=DataInterface() {propertyValues {
          AlphaCurve {keys [<at 0 1 RCIM_Cubic>]}
          BlueCurve {keys [<at 0.1619 3 RCIM_Cubic>,<at 0.3675 0.7386 RCIM_Cubic>]}
          GreenCurve {keys [<at 0.1619 1.7072 RCIM_Cubic>,<at 0.3675 0.7386 RCIM_Cubic>]}
          RedCurve {keys [<at 0.1619 1.6436 RCIM_Cubic>,<at 0.3675 0.7386 RCIM_Cubic>]}
         }}))
     )
     Float=&FloatFromCurve(FloatCurve=DataInterface() {propertyValues {Curve {keys [
          {interpMode=RCIM_Cubic, leaveTangentWeight=0.3728, tangentMode=RCTM_User, time=0.1, value=3}
          {arriveTangent=-2.7662, interpMode=RCIM_Cubic, leaveTangent=-2.7662, tangentMode=RCTM_User, time=1, value=1}
         ]}}})
    )
    ScaleA=true
    ScaleRGB=true
   )
  ]
  rendererClasses [&NiagaraSpriteRendererProperties]
  renderers [
   {rendererClass=&NiagaraSpriteRendererProperties}
   {rendererClass=&NiagaraSpriteRendererProperties, rendererIndex=1}
  ]
 }
}
summary {
 systemName=NS_Pickaxe_MicStand_AnvilAlarmRobin_Impact
 emitters [
  {
   emitterName=FN_PhysMatDebris
   rendererClasses [&NiagaraSpriteRendererProperties, &NiagaraMeshRendererProperties]
  }
  {emitterName=Hearts, rendererClasses [&NiagaraSpriteRendererProperties]}
  {emitterName=Sparks, rendererClasses [&NiagaraSpriteRendererProperties]}
  {emitterName=Smoke, rendererClasses [&NiagaraSpriteRendererProperties]}
  {emitterName=Spikes001, rendererClasses [&NiagaraSpriteRendererProperties]}
  {
   emitterName=Core_Flash001
   rendererClasses [
    &NiagaraMeshRendererProperties
    &NiagaraSpriteRendererProperties
    &NiagaraLightRendererProperties
   ]
  }
  {emitterName=Sparks001, rendererClasses [&NiagaraSpriteRendererProperties]}
 ]
 userVariables [
  User.bIsFrontEnd {type=NiagaraBool}
  User.bIsFrontEndPreview {type=NiagaraBool}
  User.PhysicalSurface {type=NiagaraInt32}
 ]
}
